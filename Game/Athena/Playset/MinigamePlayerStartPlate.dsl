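; Creative player spawn plate device that places a team-assignable player start for minigames.
; Option changes push team, priority, island-start and name-lock settings to a child FortPlayerStartCreative, with enabled state also following the minigame phase.
; Cosmetic paths drive the team number and enabled-color emissive through dynamic materials.
; The authority broadcasts On Player Spawned for players who claimed the pad, with a cached controller replayed at game start.
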(event EventOnPlaced
  ;[0 1 0.1154] Initialization
  (Audio|PlaySoundatLocation (Variables|Default|GetPlaced%20Sound) (Transformation|GetActorLocation))
  (Visibility|SetVisibilityComponents (Variables|Default|GetAlwaysHiddenInGame) (Utilities|Array|MakeArray (Variables|Default|GetTeamNumberDisplay)) true))

(event Custom|Blend%20Emissive%20Color
  (params (Color :type Linear%20Color%20Structure))
  ;[1 0.7982 0] Blend Emissive quickly to new colour.
  ; This is currently disabled until I can get a diff without the colour embedded in it (so emissive will override properly)
  (if (Math|Color|NotEqual(LinearColor) Color (Variables|EmissiveShift|GetCurrent%20Color))
    (Variables|EmissiveShift|SetOld%20Color (Variables|EmissiveShift|GetCurrent%20Color))
    (Variables|EmissiveShift|SetCurrent%20Color Color)
    (Timeline|Colour%20Lerp PlayFromStart
      (:Update
        (Rendering|Material|SetVectorParameterValue (Variables|EmissiveShift|GetMID%20SpawnPlate) "EmissiveMultiplier" (Math|Color|Lerp(LinearColor) (Variables|EmissiveShift|GetOld%20Color) (Variables|EmissiveShift|GetCurrent%20Color) (Variables|Default|GetEmissiveAlpha)) :!class /Script/Engine.MaterialInstanceDynamic)))))

(event EventBeginPlay
  (bind _fortminigameplayerstart (Variables|Default|GetFortMinigamePlayerStart))
  (bind _toyoptionscomponent (Variables|Default|GetToyOptionsComponent))
  ;[0 1 0.1154] Initialization
  (|CallParentFunction)
  (bind _asfort_player_start_creative (Utilities|Casting|CastToFortPlayerStartCreative (Variables|/Script/Engine.ChildActorComponent|GetChildActor (Variables|Default|GetPlayerStartCreative))))
  (Variables|/Script/FortniteGame.FortMinigamePlayerStartComponent|SetPlayerStarts (Utilities|Array|MakeArray _asfort_player_start_creative) _fortminigameplayerstart)
  (Variables|/Script/FortniteGame.FortPlayerStartCreative|SetPlayerStartTags (Variables|/Script/FortniteGame.FortMinigameLogicComponent|GetMinigameTags _fortminigameplayerstart) _asfort_player_start_creative)
  (CallFunction|SetEnabledState)
  (OptionData|AssignOnAnyPropertyChanged _toyoptionscomponent (delegate Custom|OnOptionChanged))
  (OptionData|BindEventtoOnPropertiesLoaded _toyoptionscomponent (delegate Custom|OnOptionChanged))
  (CallFunction|ReinitializeSpawnSettings))

(event Custom|OnOptionChanged
  ;[0 1 0.1154] Initialization
  (CallFunction|SetEnabledState)
  (CallFunction|ReinitializeSpawnSettings))

(event Custom|Enable%20Event
  (params (Instigator :type Controller%20Object%20Reference))
  ;[0.6648 0 1] Gameplay Functions
  (Variables|Default|SetEnabled true))

(event Custom|UpdateTeamAssignmentDisplay
  ;[1 1 1] Updates the display of which team the spawner is assigned to
  (Networking|CanExecuteCosmeticEvents
    (:True
      (Rendering|Material|SetScalarParameterValue (Variables|Default|GetMID%20Team%20Assignment) "Score" (Material|TranslateTeamValuetoMaterialScoreParam) :!class /Script/Engine.MaterialInstanceDynamic))))

(event Custom|Reinitialize%20Spawn%20Settings
  ;[1 1 1] Setting all the appropriate values
  (Networking|CanExecuteCosmeticEvents
    (:True
      (CallFunction|InitializeTeamAssignmentDisplay)
      (CallFunction|UpdateTeamAssignmentDisplay)
      (CallFunction|UpdateEmissive))
    (:False
      (CallFunction|UpdatePriorityGroup)
      (CallFunction|UpdateUseasIslandStart)
      (CallFunction|UpdateTeamSetting)
      (CallFunction|UpdateTeamAssignmentDisplay)
      (CallFunction|UpdatePlayerStartSubstring)
      (CallFunction|UpdateIsEnabled))))

(event Custom|Disable%20Event
  (params (Instigator :type Controller%20Object%20Reference))
  ;[0.6648 0 1] Gameplay Functions
  (Variables|Default|SetEnabled false))

(event MinigameStateChangedDelegate(FortMinigameProgress)
  (params (Minigame :type Fort%20Minigame%20Object%20Reference) (NewMinigameState :type EFortMinigameState%20Enum :default "PreGame"))
  ;[0 0.432 1] Update on Minigame Progress
  (Variables|Default|SetCurrent%20Phase NewMinigameState)
  (CallFunction|SetEnabledState))

(event OnAddedtoMinigame(FortMinigameProgress)
  (params (Minigame :type Fort%20Minigame%20Object%20Reference))
  (CallFunction|BindEventtoMinigamePlayerSpawn))

(event OnRemovedfromMinigame(FortMinigameProgress)
  (params (Minigame :type Fort%20Minigame%20Object%20Reference))
  (CallFunction|UnbindEventtoMinigamePlayerSpawn))

(event MinigameStartedDelegate(FortMinigameProgress)
  (Networking|SwitchHasAuthority
    (:Authority
      (bind _controllertransmitingongamestart (Variables|Default|GetControllerTransmitingOnGameStart)
        (:then
          (CallFunction|TransmitonPlayerSpawned _controllertransmitingongamestart))
        (:else)))))

(event Custom|OnMinigamePlayerSpawned
  (params (RespawningPlayer :type Controller%20Object%20Reference))
  (Utilities|IsValid RespawningPlayer
    (:Is%20Valid
      (bind _isinprogress (Utilities|IsMinigameStateinProgress)
        (:ValidMinigame
          (if _isinprogress
            (CallFunction|TransmitonPlayerSpawned RespawningPlayer)
            (else
              (CallFunction|ValidateandCacheTransmitData RespawningPlayer))))))))

(event MinigameEndedDelegate(FortMinigameProgress)
  ;[0 0.432 1] Update on Minigame Progress
  (CallFunction|SetEnabledState))

(fn BindEventToMinigamePlayerSpawn
  (Networking|SwitchHasAuthority
    (:Authority
      (bind _currentminigame (Variables|/Script/FortniteGame.FortMinigameLogicComponent|GetCurrentMinigame (Variables|Default|GetFortMinigameProgress))
        (:then
          (Minigame|BindEventtoOnMinigamePlayerSpawned _currentminigame (EventDispatchers|CreateEvent :!event OnMinigamePlayerSpawned)))
        (:else)))))

(fn CurrentPhaseIs
  (params (Phase :type EFortMinigameState%20Enum :default "PreGame"))
  ;[1 1 1] Basic Check function to neaten up main graphs.
  (return (Utilities|Enum|Equal(Enum) Phase (Variables|Default|GetCurrent%20Phase))))

(fn GetVisualComponents
  (return (Utilities|Array|MakeArray (Variables|Debug|GetStaticMeshComponent) (Variables|Default|GetPawnCollider))))

(fn InitializeTeamAssignmentDisplay
  (Variables|Default|GetMID%20Team%20Assignment
    (:then
      _)
    (:else
      (bind _returnvalue (Rendering|Material|CreateDynamicMaterialInstance (Variables|Default|GetTeamNumberDisplay) 0 (Variables|Default|GetTeamAssignmentMat) "None" :!class /Script/Engine.PrimitiveComponent))
      (Variables|Default|SetMID%20Team%20Assignment _returnvalue)))
  (Rendering|Material|SetScalarParameterValue (Variables|Default|GetMID%20Team%20Assignment) "Score" (Material|TranslateTeamValuetoMaterialScoreParam) :!class /Script/Engine.MaterialInstanceDynamic))

(fn IsVisibleDuringMinigame
  (return (Variables|Default|GetMeshVisibleDuringMinigameOption)))

(fn IsClaimedBy
  (params (PlayerController :type Controller%20Object%20Reference))
  (Utilities|IsValid PlayerController
    (:Is%20Valid
      (bind _asfort_player_start_creative (Utilities|Casting|CastToFortPlayerStartCreative (Variables|/Script/Engine.ChildActorComponent|GetChildActor (Variables|Default|GetPlayerStartCreative)))
        (:then
          (bind _asfort_player_state (Utilities|Casting|CastToFortPlayerState (Variables|/Script/Engine.Controller|GetPlayerState PlayerController))
            (:then
              (return (Class|FortPlayerStartCreative|IsClaimedbyPlayer _asfort_player_start_creative (PlayerState|GetUniqueID _asfort_player_state))))
            (:CastFailed
              _)))
        (:CastFailed
          _))
      (return false))
    (:Is%20Not%20Valid
      _)))

(fn IsInsideSameVolume
  (params (PlayerState :type Player%20State%20Object%20Reference))
  (Utilities|IsActorRelevanttoMe PlayerState
    (:IsRelevant
      (return true))
    (:NotRelevant
      (return false))))

(fn OnRep_Enabled
  (CallFunction|ReinitializeSpawnSettings))

(fn SetEnabledState
  (switch int (Variables|UserKnobs|GetEnabledIndex)
    (:Default
      ;[1 1 1] None
      (Variables|Default|SetEnabled false))
    (:0
      (Variables|Default|SetEnabled true))
    (:1
      ;[1 1 1] Pre-Game Only
      (Variables|Default|SetEnabled (Utilities|Array|ContainsItem (Variables|Default|GetPregamePhases) (Variables|Default|GetCurrent%20Phase))))
    (:2
      ;[1 1 1] Gameplay Only
      (Variables|Default|SetEnabled (Utilities|Array|ContainsItem (Variables|Default|GetGameplay%20Phases) (Variables|Default|GetCurrent%20Phase)))))
  (CallFunction|UpdateIsEnabled))

(fn TranslateTeamValueToMaterialScoreParam
  (if (Knobs|Team|IsTeamNone (Variables|UserKnobs|GetPlayerTeam))
    (return -1.0)
    (elif (Knobs|Team|IsTeamAny (Variables|UserKnobs|GetPlayerTeam))
      (return 0.0)
      (else
        (return (Math|Conversions|ToFloat(Integer) (Knobs|Team|GetTeamValue (Variables|UserKnobs|GetPlayerTeam))))))))

(fn TransmitOnPlayerSpawned
  (params (PlayerController :type Controller%20Object%20Reference))
  (Networking|SwitchHasAuthority
    (:Authority
      (bind _playercontroller (Variables|GetPlayerController)
        (:then
          (bind _playerstate (Variables|/Script/Engine.Controller|GetPlayerState _playercontroller)
            (:then
              (LocalVariables|SetTransmitterPlayerState _playerstate)
              (if (and (CallFunction|IsClaimedBy (Variables|GetPlayerController)) (CallFunction|IsInsideSameVolume (LocalVariables|GetTransmitterPlayerState)))
                (GameplayEvent|BroadcastEvent (Variables|UserOptions|Events|GetOn%20Player%20Spawned) 0 (Utilities|Struct|MakeGameplayEventSpawned (Variables|GetPlayerController)))
                (Variables|Default|SetControllerTransmitingOnGameStart 0)))
            (:else)))
        (:else)))))

(fn UnbindEventToMinigamePlayerSpawn
  (Networking|SwitchHasAuthority
    (:Authority
      (bind _currentminigame (Variables|/Script/FortniteGame.FortMinigameLogicComponent|GetCurrentMinigame (Variables|Default|GetFortMinigameProgress))
        (:then
          (Minigame|UnbindEventfromOnMinigamePlayerSpawned _currentminigame (EventDispatchers|CreateEvent :!event OnMinigamePlayerSpawned)))
        (:else)))))

(fn UpdateEmissive
  (Networking|CanExecuteCosmeticEvents
    (:True
      (Variables|EmissiveShift|GetMID%20SpawnPlate
        (:then
          _)
        (:else
          (bind _returnvalue (Rendering|Material|CreateDynamicMaterialInstance (Variables|Debug|GetStaticMeshComponent) 0 0 "None" :!class /Script/Engine.PrimitiveComponent))
          (Variables|EmissiveShift|SetMID%20SpawnPlate _returnvalue)
          (Rendering|Material|SetTextureParameterValue (Variables|EmissiveShift|GetMID%20SpawnPlate) "EmissiveTexture" (Variables|EmissiveShift|GetEmissive%20Texture))))
      (CallFunction|BlendEmissiveColor (Math|Color|SelectColor (Variables|Default|GetEnabledColor) (Variables|Default|GetDisabledColor) (and (Variables|Default|GetEnabled) (not (Knobs|Team|IsTeamNone (Variables|UserKnobs|GetPlayerTeam)))))))))

(fn UpdateIsEnabled
  (bind _asfort_player_start_creative (Utilities|Casting|CastToFortPlayerStartCreative (Variables|/Script/Engine.ChildActorComponent|GetChildActor (Variables|Default|GetPlayerStartCreative))))
  (Variables|/Script/FortniteGame.FortPlayerStartCreative|SetbIsEnabled (Variables|Default|GetEnabled) _asfort_player_start_creative))

(fn UpdatePlayerStartSubstring
  (LocalVariables|SetLockedStringToUse (Utilities|String|TrimTrailing (Utilities|String|Trim (Variables|Default|GetLockedPlayerNameSubstring))))
  (for _array_element (Variables|/Script/FortniteGame.FortMinigamePlayerStartComponent|GetPlayerStarts (Variables|Default|GetFortMinigamePlayerStart))
    (Variables|/Script/FortniteGame.FortPlayerStartCreative|SetLockedPlayerNameSubstring (LocalVariables|GetLockedStringToUse) _array_element)))

(fn UpdatePriorityGroup
  (bind (_playerstarttags _portalindex _applicableteam _applicableclass _buseasislandstart _prioritygroup) (Class|GetClassDefaults "/Script/FortniteGame.FortPlayerStartCreative"))
  (bind _playerstartcreative (Variables|Default|GetPlayerStartCreative))
  (bind _childactor (Variables|/Script/Engine.ChildActorComponent|GetChildActor _playerstartcreative))
  (if (< (Variables|Default|GetPriorityGroupSetting) 1)
    (bind _asfort_player_start_creative (Utilities|Casting|CastToFortPlayerStartCreative _childactor))
    (Variables|/Script/FortniteGame.FortPlayerStartCreative|SetPriorityGroup _prioritygroup _asfort_player_start_creative)
    (else
      (bind _asfort_player_start_creative_2 (Utilities|Casting|CastToFortPlayerStartCreative _childactor))
      (Variables|/Script/FortniteGame.FortPlayerStartCreative|SetPriorityGroup (Variables|Default|GetPriorityGroupSetting) _asfort_player_start_creative_2))))

(fn UpdateTeamSetting
  (bind _playerstartcreative (Variables|Default|GetPlayerStartCreative))
  (bind _childactor (Variables|/Script/Engine.ChildActorComponent|GetChildActor _playerstartcreative))
  (Utilities|PlayerTeamifEnabled
    (:NoneTeam
      (bind _asfort_player_start_creative (Utilities|Casting|CastToFortPlayerStartCreative _childactor))
      (Variables|/Script/FortniteGame.FortPlayerStartCreative|SetApplicableTeam (Utilities|Enum|LiteralenumEFortTeam "Spectator") _asfort_player_start_creative))
    (:Any
      (bind _asfort_player_start_creative_2 (Utilities|Casting|CastToFortPlayerStartCreative _childactor))
      (Variables|/Script/FortniteGame.FortPlayerStartCreative|SetApplicableTeam (Utilities|Enum|LiteralenumEFortTeam "HumanCampaign") _asfort_player_start_creative_2))
    (:ValidTeam
      (bind _asfort_player_start_creative_3 (Utilities|Casting|CastToFortPlayerStartCreative _childactor))
      (Variables|/Script/FortniteGame.FortPlayerStartCreative|SetApplicableTeam (Teams|IndextoTeam (Knobs|Team|GetTeamValueasIndex (Variables|UserKnobs|GetPlayerTeam))) _asfort_player_start_creative_3))))

(fn UpdateUseAsIslandStart
  (bind _asfort_player_start_creative (Utilities|Casting|CastToFortPlayerStartCreative (Variables|/Script/Engine.ChildActorComponent|GetChildActor (Variables|Default|GetPlayerStartCreative))))
  (Variables|/Script/FortniteGame.FortPlayerStartCreative|SetbUseAsIslandStart (Variables|Default|GetUseAsIslandStart) _asfort_player_start_creative))

(fn ConstructionScript)

(fn ValidateAndCacheTransmitData
  (params (FortPlayerControllerAthena :type Controller%20Object%20Reference))
  (bind _fortplayercontrollerathena (Variables|GetFortPlayerControllerAthena)
    (:then
      (bind _currentminigame (Variables|/Script/FortniteGame.FortMinigameLogicComponent|GetCurrentMinigame (Variables|Default|GetFortMinigameProgress))
        (:then
          (if (Minigame|IsinProgress _currentminigame)
            (if (CallFunction|IsClaimedBy _fortplayercontrollerathena)
              (Variables|Default|SetControllerTransmitingOnGameStart _fortplayercontrollerathena))))
        (:else)))
    (:else)))
