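; Minigame trigger plate that tracks per-player progress while pawns stand in its box, which only has query collision during a minigame.
; Overlap begin and end start or abort progress on FortMinigameProgress, skipping players in Locked Players.
; Finishing locks that player until the next minigame start and, on remotes, hides the marker and fires particles.
; A timeline, rate-scaled to the activation time, drives the marker material's TimeLeft parameter as a cosmetic countdown.
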(event EventOnPlaced
  (bind _fortminigameprogress (Variables|Default|GetFortMinigameProgress))
  (Collision|SetCollisionEnabled (Variables|Default|GetBox) "NoCollision")
  (Networking|CanExecuteCosmeticEvents
    (:True
      (Development|SetHiddeninGame (Variables|Default|GetMarker) false false)
      (if (Minigame|HasActivationTime _fortminigameprogress)
        (Components|Timeline|SetPlayRate (Variables|MinigameTriggerPlate|GetMarkerTimeline) (/ (Components|Timeline|GetTimelineLength (Variables|MinigameTriggerPlate|GetMarkerTimeline)) (Variables|/Script/FortniteGame.FortMinigameProgressComponent|GetActivationTime _fortminigameprogress)))))))

(event MinigameStartedDelegate(FortMinigameProgress)
  (Collision|SetCollisionEnabled (Variables|Default|GetBox) "QueryOnly")
  (Utilities|Array|Clear (Variables|Default|GetLocked%20Players))
  (Networking|SwitchHasAuthority
    (:Remote
      (Networking|CanExecuteCosmeticEvents
        (:True
          (Development|SetHiddeninGame (Variables|Default|GetMarker) false false)
          (Rendering|Material|SetScalarParameterValueonMaterials (Variables|Default|GetMarker) "TimeLeft" 1.0))))))

(event MinigameEndedDelegate(FortMinigameProgress)
  (Collision|SetCollisionEnabled (Variables|Default|GetBox) "NoCollision")
  (Networking|SwitchHasAuthority
    (:Remote
      (Networking|CanExecuteCosmeticEvents
        (:True
          (Development|SetHiddeninGame (Variables|Default|GetMarker) false false))))))

(event ProgressStartedDelegate(FortMinigameProgress)
  (params (PlayerState :type Player%20State%20Object%20Reference) (Progress :type Fort%20Minigame%20Progress%20Component%20Object%20Reference))
  (Networking|CanExecuteCosmeticEvents
    (:True
      (if (Minigame|HasActivationTime Progress)
        (bind (_direction _alpha) (Timeline|MarkerTimeline PlayFromStart)
          (:Update
            (Rendering|Material|SetScalarParameterValueonMaterials (Variables|Default|GetMarker) "TimeLeft" _alpha)))))))

(event ProgressFinishedDelegate(FortMinigameProgress)
  (params (PlayerState :type Player%20State%20Object%20Reference) (Progress :type Fort%20Minigame%20Progress%20Component%20Object%20Reference))
  (Utilities|Array|AddUnique (Variables|Default|GetLocked%20Players) PlayerState)
  (Networking|SwitchHasAuthority
    (:Remote
      (Networking|CanExecuteCosmeticEvents
        (:True
          (Development|SetHiddeninGame (Variables|Default|GetMarker) true false)
          (Components|Activation|Activate (Variables|Default|GetParticleSystem) true)
          (Timeline|MarkerTimeline Stop))))))

(event ProgressAbortedDelegate(FortMinigameProgress)
  (params (PlayerState :type Player%20State%20Object%20Reference) (Progress :type Fort%20Minigame%20Progress%20Component%20Object%20Reference))
  (bind _marker (Variables|Default|GetMarker))
  (Networking|CanExecuteCosmeticEvents
    (:True
      (Development|SetHiddeninGame _marker false false)
      (Rendering|Material|SetScalarParameterValueonMaterials _marker "TimeLeft" 1.0)
      (Timeline|MarkerTimeline Stop))))

(event OnComponentBeginOverlap(Box)
  (params (OverlappedComponent :type Primitive%20Component%20Object%20Reference) (OtherActor :type Actor%20Object%20Reference) (OtherComp :type Primitive%20Component%20Object%20Reference) (OtherBodyIndex :type Integer) (bFromSweep :type Boolean) (SweepResult :type Hit%20Result%20Structure%20%28by%20ref%29))
  (bind _aspawn (Utilities|Casting|CastToPawn OtherActor))
  (Utilities|IsPawnRelevant _aspawn
    (:True
      (bind _playerstate (Variables|/Script/Engine.Pawn|GetPlayerState _aspawn)
        (:then
          (if (not (Utilities|Array|ContainsItem (Variables|Default|GetLocked%20Players) _playerstate))
            (Minigame|StartProgress (Variables|Default|GetFortMinigameProgress) _playerstate)))
        (:else)))))

(event OnComponentEndOverlap(Box)
  (params (OverlappedComponent :type Primitive%20Component%20Object%20Reference) (OtherActor :type Actor%20Object%20Reference) (OtherComp :type Primitive%20Component%20Object%20Reference) (OtherBodyIndex :type Integer))
  (bind _aspawn (Utilities|Casting|CastToPawn OtherActor))
  (Utilities|IsPawnRelevant _aspawn
    (:True
      (bind _playerstate (Variables|/Script/Engine.Pawn|GetPlayerState _aspawn)
        (:then
          (if (not (Utilities|Array|ContainsItem (Variables|Default|GetLocked%20Players) _playerstate))
            (Minigame|AbortProgress (Variables|Default|GetFortMinigameProgress) _playerstate)))
        (:else)))))

(fn ConstructionScript)
