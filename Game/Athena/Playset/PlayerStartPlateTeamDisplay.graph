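# Unlit translucent material for the player start plate team display, drawing a numeric Score as a tinted dot-matrix readout.
# Floor math offsets UVs into a numbers atlas to pick digits, with horizontal masks separating the digit slots.
# Scores of 100 and above switch from a two-digit to a three-digit layout and triple the circle texture's tiling.
# Color tints the result, and Win blends in a time-driven sine pulse.
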Material /Game/Athena/Playset/PlayerStartPlateTeamDisplay
graph {
 properties {blendMode=BLEND_Translucent, pixelDepthOffsetMode=PDOM_Legacy, shadingModel=MSM_Unlit, usedShadingModels=Unlit}
 nodes [
  n1 = VectorParameter() {parameterName=Color, defaultValue=<rgba 1 1 1 1>}
  n2 = TextureCoordinate() {desc=two digits, uTiling=20, vTiling=14}
  n3 = NamedRerouteUsage()
  n4 = Saturate(None=Floor(None=Multiply(A=@n3) {constB=0.01}))
  n5 = LinearInterpolate(A=@n2, B=Multiply(A=@n2) {constB=3}, Alpha=@n4)
  n6 = TextureCoordinate() {uTiling=0.2}
  n7 = ComponentMask(None=@n6) {r=true}
  n8 = NamedRerouteDeclaration(None=ScalarParameter() {parameterName=Score}) {name=score, nodeColor=<rgba 0.5 0 1 1>}
  n9 = Floor(None=Multiply(A=@n8) {constB=0.1})
  n10 = ComponentMask(None=@n6) {g=true}
  n11 = AppendVector(A=Add(A=@n7, B=Multiply(A=@n9) {constB=0.1}), B=@n10)
  n12 = Multiply(A=Floor(None=@n8)) {constB=0.1}
  n13 = AppendVector(A=Add(A=Add(A=@n7, B=@n12)) {constB=-0.1}, B=@n10)
  n14 = ComponentMask(None=Subtract(A=TextureCoordinate()) {constB=0.5}) {r=true}
  n15 = Saturate(None=LinearInterpolate(
    A=TextureSample(UVs=@n11) {desc=10s digit, texture=/Game/Environments/Props/Street/Textures/numbers}.R
    B=TextureSample(UVs=@n13) {desc=1s digits, texture=&numbers}.R
    Alpha=Ceil(None=@n14)
   ))
  n16 = NamedRerouteUsage()
  n17 = Add(A=NamedRerouteUsage(), B=Floor(None=Multiply(A=@n16) {constB=0.1}))
  n18 = Multiply(A=Floor(None=Multiply(A=@n16) {constB=0.01})) {constB=0.1}
  n19 = TextureSample(UVs=AppendVector(A=Add(A=@n17, B=@n18) {constB=0.6}, B=NamedRerouteUsage())) {desc=100s digit, texture=&numbers}
  n20 = NamedRerouteUsage()
  n21 = Floor(None=@n16)
  n22 = Add(A=Add(A=NamedRerouteUsage(), B=Multiply(A=@n21) {constB=0.1})) {constB=-0.2}
  n23 = NamedRerouteUsage()
  n24 = Multiply(
   A=TextureSample(UVs=AppendVector(A=@n22, B=@n23)) {desc=1s digit, texture=&numbers}.R
   B=NamedRerouteUsage()
  )
  n25 = Floor(None=Multiply(A=@n16) {constB=0.1})
  n26 = Add(A=Add(A=NamedRerouteUsage(), B=Multiply(A=@n25) {constB=0.1})) {constB=-0.1}
  n27 = NamedRerouteUsage()
  n28 = Multiply(
   A=TextureSample(UVs=AppendVector(A=@n26, B=@n27)) {desc=10s digit, texture=&numbers}.R
   B=NamedRerouteUsage()
  )
  n29 = Saturate(None=Add(A=Multiply(A=@n19.R, B=@n20), B=Add(A=@n24, B=@n28)))
  n30 = Multiply(
   A=@n1.RGB
   B=Multiply(
    A=TextureSample(UVs=@n5) {texture=/Game/Environments/Props/Street/Textures/CircleTexture}.R
    B=LinearInterpolate(A=@n15, B=@n29, Alpha=@n4) {constAlpha=0}
   )
  )
  n31 = Multiply(A=Time()) {constB=2}
  LinearInterpolate(A=@n30, B=Multiply(A=@n30, B=Sine(None=@n31)), Alpha=ScalarParameter() {parameterName=Win})
  n32 = TextureCoordinate()
  n33 = Subtract(A=@n32) {constB=0.3}
  n34 = OneMinus(None=Ceil(None=ComponentMask(None=@n33) {r=true}))
  NamedRerouteDeclaration(None=@n34) {name=100mask, nodeColor=<rgba 0 1 0.3672 1>}
  n35 = Ceil(None=ComponentMask(None=Subtract(A=@n32) {constB=0.65}) {r=true})
  NamedRerouteDeclaration(None=@n35) {name=1mask, nodeColor=<rgba 0.6172 0 1 1>}
  NamedRerouteDeclaration(None=OneMinus(None=Add(A=@n34, B=@n35))) {name=10mask, nodeColor=<rgba 1 0 0.1172 1>}
  n36 = TextureCoordinate() {uTiling=0.3}
  NamedRerouteDeclaration(None=ComponentMask(None=@n36) {g=true}) {name=greenChannel, nodeColor=<rgba 0.0335 0.4427 0.1015 1>}
  NamedRerouteDeclaration(None=ComponentMask(None=@n36) {r=true}) {name=redChannel, nodeColor=<rgba 1 0.026 0.026 1>}
  StaticSwitchParameter(True=@n1.RGB, False=Constant()) {group=Color, parameterName=Use Color}
 ]
}
