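# Translucent, unlit, two-sided material for a player start plate team display, showing a two-digit Score in a tintable Color.
# Left and right halves pick tens and ones digits by offsetting UVs into a numbers atlas.
# A tiled circle texture masks the digits into a dot-matrix look.
# For camera facing, the graph holds an AlignMeshToTheCamera call and an atan2-driven RotateAboutAxis around an offsettable pivot.
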Material /Game/Athena/Playset/PlayerStartPlateTeamDisplay_PlayerFacing_2Axis
graph {
 properties {blendMode=BLEND_Translucent, pixelDepthOffsetMode=PDOM_Legacy, shadingModel=MSM_Unlit, twoSided=true, usedShadingModels=Unlit}
 nodes [
  MaterialFunctionCall!(
   WorldPosition
   PivotLocation
   WS Normals
   Custom Object Basis 1=Transform(None=Constant3Vector() {constant=<rgba 1 0 0 0>}) {transformSourceType=TRANSFORMSOURCE_Local}
   Custom Object Basis 2=Transform(None=Constant3Vector() {constant=<rgba 0 0 -1 0>}) {transformSourceType=TRANSFORMSOURCE_Local}
   Custom Object Basis 3=Transform(None=Constant3Vector() {constant=<rgba 0 1 0 0>}) {transformSourceType=TRANSFORMSOURCE_Local}
  ) {materialFunction=/Engine/Functions/Engine_MaterialFunctions02/WorldPositionOffset/AlignMeshToTheCamera}
  n1 = VectorParameter() {parameterName=Color, defaultValue=<rgba 1 1 1 1>}
  n2 = TextureCoordinate() {uTiling=20, vTiling=14}
  n3 = TextureCoordinate() {uTiling=0.2}
  n4 = ComponentMask(None=@n3) {r=true}
  n5 = ScalarParameter() {defaultValue=1, parameterName=Score}
  n6 = Multiply(A=Floor(None=Multiply(A=@n5) {constB=0.1})) {constB=0.1}
  n7 = ComponentMask(None=@n3) {g=true}
  n8 = TextureSample(UVs=AppendVector(A=Add(A=@n4, B=@n6), B=@n7)) {texture=/Game/Environments/Props/Street/Textures/numbers}
  n9 = Add(A=@n4, B=Multiply(A=Floor(None=@n5)) {constB=0.1})
  n10 = TextureSample(UVs=AppendVector(A=Add(A=@n9) {constB=-0.1}, B=@n7)) {texture=&numbers}
  n11 = TextureCoordinate()
  n12 = Ceil(None=ComponentMask(None=Subtract(A=@n11) {constB=0.5}) {r=true})
  Multiply(
   A=@n1.RGB
   B=Multiply(
    A=TextureSample(UVs=@n2) {texture=/Game/Environments/Props/Street/Textures/CircleTexture}.R
    B=LinearInterpolate(A=@n8.R, B=@n10.R, Alpha=@n12)
   )
  )
  Reroute(None=@n1.RGB)
  n13 = CameraPositionWS()
  n14 = Constant3Vector() {constant=<rgba 1 1 0 0>}
  MaterialExpressionNormalize_0 = Normalize(VectorInput=Subtract(A=ObjectPositionWS(), B=Multiply(A=@n13, B=@n14)))
  n15 = Custom!(In=@MaterialExpressionNormalize_0) {
   code="float2 output; \r\nfloat2 Input1; \r\noutput = atan2 (In.y,In.x); \r\nreturn (output);"
   desc="float2 output; \r\nfloat2 Input1; \r\noutput = atan2 (In.y,In.x); \r\nreturn (output);"
   outputType=CMOT_Float2
   inputs [{inputName=In, input {expression=&self:MaterialExpressionNormalize_0}}]
  }
  n16 = Divide(A=@n15) {constB=3.142}
  RotateAboutAxis(
   NormalizedRotationAxis=ObjectOrientation()
   RotationAngle=Add(A=Divide(A=@n16)) {constB=0.25}
   PivotPoint=Add(
    A=MaterialFunctionCall() {materialFunction=&ObjectPivotPoint}.Object Pivot Location
    B=VectorParameter() {group=WPO, parameterName=Pivot Offset, defaultValue=<rgba 0 0 0 1>}.RGB
   )
   Position=WorldPosition().XYZ
  )
 ]
}
