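; Interactable minigame control panel that starts the minigame in pre-game and abandons it while in progress.
; Interaction prompts follow the minigame state, and interaction is blocked during warmup or when no minigame is found.
; On minigame start and end, the authority sets CurrentColor to the active or inactive color.
; Its OnRep function applies that color to the Alarm Clock component's inner glow material parameter.
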(event Interaction|EventBlueprintOnInteract
  (params (InteractingPawn :type Fort%20Pawn%20Object%20Reference) (InteractionBeingAttempted :type EInteractionBeingAttempted%20Enum :default "FirstInteraction"))
  (bind _minigame (Utilities|GetMinigame)
    (:Valid
      (Utilities|SwitchCurrentState _minigame
        (:Pre%20Game
          (Minigame|StartMinigame _minigame))
        (:In%20Progress
          (Minigame|EndGame _minigame "AbandonGame"))))
    (:Not%20Valid
      (Development|PrintString "SimpleMinigameControlPanel: Failed to get minigame"))))

(event MinigameStartedDelegate(FortMinigameLogic)
  (Networking|SwitchHasAuthority
    (:Authority
      (Variables|Default|SetCurrent%20Color (Variables|Default|GetActive%20Color)))))

(event MinigameEndedDelegate(FortMinigameLogic)
  (Networking|SwitchHasAuthority
    (:Authority
      (Variables|Default|SetCurrent%20Color (Variables|Default|GetInacctive%20Color)))))

(event EventOnDeathServer
  (params (Damage :type Float%20%28single-precision%29) (DamageTags :type Gameplay%20Tag%20Container%20Structure%20%28by%20ref%29) (Momentum :type Vector :default "0, 0, 0") (HitInfo :type Hit%20Result%20Structure%20%28by%20ref%29) (InstigatedBy :type Controller%20Object%20Reference) (DamageCauser :type Actor%20Object%20Reference) (EffectContext :type Gameplay%20Effect%20Context%20Handle%20Structure)))

(fn BlueprintCanInteract
  (params (InteractingPawn :type Fort%20Pawn%20Object%20Reference) (InteractionBeingAttempted :type EInteractionBeingAttempted%20Enum :default "FirstInteraction") (InteractionType :type TInteractionType%20Enum :default "IT_NoInteraction"))
  (bind _minigame (Utilities|GetMinigame)
    (:Valid
      (Utilities|SwitchCurrentState _minigame
        (:Pre%20Game
          _)
        (:Warmup
          (return false))
        (:In%20Progress
          _)
        (:Post%20Round%20End
          _)
        (:Post%20Game%20End
          _)
        (:Post%20Game%20Abandon
          _))
      (return true))
    (:Not%20Valid
      _)))

(fn BlueprintGetFailedInteractionString
  (params (InteractingPawn :type Fort%20Pawn%20Object%20Reference) (InteractionBeingAttempted :type EInteractionBeingAttempted%20Enum :default "FirstInteraction"))
  (Utilities|GetMinigame
    (:Valid
      (return "NSLOCTEXT(\"[296AC82048D2C04A0D0FF58560F425E4]\", \"4AD7B42F42689CE2F0D662AC99DF0C1D\", \"Please Wait\")"))
    (:Not%20Valid
      (return "NSLOCTEXT(\"[296AC82048D2C04A0D0FF58560F425E4]\", \"9E3E02304A55DA1C111D0EA6CC3BFA02\", \"No Game Set\")"))))

(fn BlueprintGetInteractionString
  (params (InteractingPawn :type Fort%20Pawn%20Object%20Reference) (InteractionBeingAttempted :type EInteractionBeingAttempted%20Enum :default "FirstInteraction"))
  (bind _minigame (Utilities|GetMinigame)
    (:Valid
      (Utilities|SwitchCurrentState _minigame
        (:Pre%20Game
          (return "NSLOCTEXT(\"[296AC82048D2C04A0D0FF58560F425E4]\", \"781781DE4B05E35D90219496D9377EE3\", \"Start Game\")"))
        (:In%20Progress
          (return "NSLOCTEXT(\"[296AC82048D2C04A0D0FF58560F425E4]\", \"1E37C6AF4D6EA4B86A7142AAC72CB221\", \"Abandon Game\")"))))
    (:Not%20Valid
      (return "NSLOCTEXT(\"[296AC82048D2C04A0D0FF58560F425E4]\", \"640601DC4283FAC6D52FBBA859CFF787\", \"No Game\")"))))

(fn OnRep_CurrentColor
  (Rendering|Material|SetVectorParameterValueonMaterials (Variables|Default|GetAlarm%20Clock) "Glow Color Inner" (Math|Conversions|ToVector(LinearColor) (Variables|Default|GetCurrent%20Color))))

(fn ConstructionScript)
